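# Simplified material for Athena arid desert roads, flagged for spline meshes and built on material attributes.
# World-space XY, scaled by MPC_Landscape, drives the arid sand function, which is multiplied by sand color and optionally passed through the desert road function.
# An optional topo pass color-matches terrain using biome and topo masks over TerrainSize-normalized world UVs, then blends in desaturated, MPC-tinted sediment as replacement base color.
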Material /Figment_S06/FN_6_20/Athena/Environments/Landscape/Material/Arid/M_AthenaDesertRoad_Simple
graph {
 properties {
  bUseMaterialAttributes=true
  bUsedWithSplineMeshes=true
  pixelDepthOffsetMode=PDOM_Legacy
  refractionMethod=RM_IndexOfRefraction
  usedShadingModels=Default Lit
 }
 nodes [
  LandscapeLayerCoords() {mappingScale=1525}
  n1 = CollectionParameter() {collection=/FigmentEnvironment/FN_6_20/Athena/Environments/Landscape/MPC/MPC_Landscape, parameterName=Arid_Sand_TexSize}
  n2 = Divide(A=WorldPosition() {worldPositionShaderOffset=WPT_ExcludeAllShaderOffsets}.XYZ, B=Multiply(A=Abs(None=@n1)) {constB=-1})
  n3 = MaterialFunctionCall!(
   In=MaterialFunctionCall!(UseRoughness=StaticBool(), UVs=ComponentMask(None=@n2) {g=true, r=true}) {materialFunction=/FigmentEnvironment/FN_6_20/Athena/Environments/Landscape/MaterialFunctions/Arid/MF_Athena_Arid_Sand}.Simple_Material
   Color=MaterialFunctionCall() {materialFunction=/FigmentEnvironment/FN_6_20/Athena/Environments/Landscape/MaterialFunctions/MF_Athena_ZSandColor}.Result
  ) {materialFunction=/FigmentEnvironment/FN_6_20/Athena/Environments/Landscape/MaterialFunctions/MF_Athena_Material_BaseMultiply}
  n4 = StaticSwitchParameter(
   True=MaterialFunctionCall!(In=@n3.Result) {materialFunction=/FigmentEnvironment/FN_6_20/Athena/Environments/Landscape/MaterialFunctions/Arid/MF_Athena_DesertRoad}.Result
   False=@n3.Result
  ) {defaultValue=true, parameterName=UseDesertRoad}
  n5 = TextureObject() {samplerType=SAMPLERTYPE_LinearColor, texture=/FigmentEnvironment/FN_6_20/Athena/Environments/Landscape/Textures/T_Athena_Terrain_Topo_Mask}
  n6 = WorldPosition()
  n7 = ScalarParameter() {defaultValue=306200, group=TerrainColorMatch, parameterName=TerrainSize}
  n8 = Divide(A=@n7)
  n9 = Divide(A=Subtract(A=ComponentMask(None=@n6.XYZ) {g=true, r=true}, B=AppendVector(A=@n8, B=@n8)), B=@n7)
  n10 = TextureSample(UVs=@n9) {
   samplerSource=SSM_Wrap_WorldGroupSettings
   samplerType=SAMPLERTYPE_LinearColor
   texture=/FigmentEnvironment/FN_6_20/Athena/Environments/Landscape/Textures/T_Athena_Terrain_Biome_M
  }
  n11 = MaterialFunctionCall!(
   Mask 2=@n5
   UV Coordinates=@n9
   RockMask
   GrassMask=Constant() {r=1}
   Arid=@n10.B
   BaseTexture=MaterialFunctionCall!(Material=@n4) {materialFunction=/Engine/Functions/MaterialLayerFunctions/MatLayerBlend_BreakBaseColor}.BaseColor
  ) {materialFunction=/FigmentEnvironment/FN_6_20/Athena/Environments/Landscape/MaterialFunctions/MF_TerrainTopoAdjustment}
  n12 = Multiply(
   A=Desaturation(
    None=@n11.Result
    Fraction=CollectionParameter() {collection=&MPC_Landscape, parameterName=Arid_EdgeDesaturation}
   )
   B=ComponentMask(None=CollectionParameter() {collection=&MPC_Landscape, parameterName=Arid_SandSediment_Tint}) {b=true, g=true, r=true}
  )
  StaticSwitchParameter(
   True=MaterialFunctionCall!(Mat_In=@n4, BaseColor=LinearInterpolate(A=@n11.Result, B=@n12, Alpha=@n11.Result)) {materialFunction=/FigmentEnvironment/FN_6_20/Athena/Environments/Landscape/MaterialFunctions/MF_Athena_ReplaceBaseColor}.Result
   False=@n4
  ) {defaultValue=true, parameterName=UseTopo}
  VertexInterpolator(VS=@n9)
 ]
}
